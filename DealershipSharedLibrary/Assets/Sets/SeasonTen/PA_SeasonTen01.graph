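# Physics asset for a Season Ten set in the dealership shared library, covering a small three-bone rig.
# Each of icon_exp_joint, jnt_13 and jnt_14 has a single capsule body, with the icon_exp_joint capsule the smallest.
# Two constraints link jnt_13 and jnt_14 directly to icon_exp_joint in a star layout, sharing identical swing and twist angle limits.
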PhysicsAsset /DealershipSharedLibrary/Assets/Sets/SeasonTen/PA_SeasonTen01
bodies {
 icon_exp_joint [{shapeType=Capsule, length=0.505, radius=0.505, rotation=<rot 0 -90 0>}]
 jnt_13 [{shapeType=Capsule, length=1.8006, radius=1.5064, center=<0.0306 0.0489 -0.0003>, rotation=<rot 0 31.6722 -89.2985>}]
 jnt_14 [{shapeType=Capsule, length=2.6113, radius=2.3054, center=<-0.0072 0.1289 0.0545>, rotation=<rot 0 32.2725 -85.3224>}]
}
constraints [
 {bone1Name=jnt_13, bone2Name=icon_exp_joint, swing1LimitDegrees=45, swing2LimitDegrees=45, twistLimitDegrees=45}
 {bone1Name=jnt_14, bone2Name=icon_exp_joint, swing1LimitDegrees=45, swing2LimitDegrees=45, twistLimitDegrees=45}
]
